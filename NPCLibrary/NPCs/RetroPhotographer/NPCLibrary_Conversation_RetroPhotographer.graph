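# Conversation database for the RetroPhotographer NPC, Lavendra, defining the dialogue shown when a player talks to them.
# The entry point leads to one Speech node with a default greeting and contextual lines.
# A prioritized introduction requires a first-ever conversation, while ambient remarks check this NPC's match conversation count.
# The speech then branches through link nodes to the shared NPCLibrary Quests and BasicServices conversation entries.
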ConversationDatabase /NPCLibrary/NPCs/RetroPhotographer/NPCLibrary_Conversation_RetroPhotographer
entry_points [{tag=Conversation.Entry.RetroPhotographer, destinations [#1]}]
nodes [
 ConversationEntryPointNode_7 {
  id=#1
  class=ConversationEntryPointNode
  outputs [#2]
  properties {entryTag {tagName=Conversation.Entry.RetroPhotographer}}
 }
 Speech_C_4 {
  id=#2
  class=Speech_C
  outputs [#3, #4]
  properties {
   nodeName=""
   generalConfig {
    defaultMessage="Oh wow, the sky is sooo beautiful!"
    contextualMessages [
     {
      message="Oh hi! Call me Lavendra."
      contextRequirements [{requirement {class=FortControllerRequirement_IsFirstEverConversationWithNPC}}]
      priority {curve {rowName=Tandem.Speech.Priority.FirstEverInteraction}}
     }
     {
      message="Oh wow, the sky is sooo beautiful!"
      contextRequirements [{requirement {
         class=FortControllerRequirement_CheckNPCMatchConversationCount
         properties {nPCUniqueIdTag {tagName=AISpawnerData.Type.Tandem.RetroPhotographer}}
        }}]
      priority {curve {rowName=None}}
     }
     {
      message="This Island is so pretty. I've taken so many amazing photos!"
      contextRequirements [{requirement {
         class=FortControllerRequirement_CheckNPCMatchConversationCount
         properties {nPCUniqueIdTag {tagName=AISpawnerData.Type.Tandem.RetroPhotographer}}
        }}]
      priority {curve {rowName=None}}
     }
     {
      message="My friends sometimes call me VeeVee. You can, if you want."
      contextRequirements [{requirement {
         class=FortControllerRequirement_CheckNPCMatchConversationCount
         properties {nPCUniqueIdTag {tagName=AISpawnerData.Type.Tandem.RetroPhotographer}}
        }}]
      priority {curve {rowName=None}}
     }
     {
      message=Try not to step on the flowers.
      contextRequirements [{requirement {
         class=FortControllerRequirement_CheckNPCMatchConversationCount
         properties {nPCUniqueIdTag {tagName=AISpawnerData.Type.Tandem.RetroPhotographer}}
        }}]
      priority {curve {rowName=None}}
     }
     {
      message=What a gorgeous day.
      contextRequirements [{requirement {
         class=FortControllerRequirement_CheckNPCMatchConversationCount
         properties {nPCUniqueIdTag {tagName=AISpawnerData.Type.Tandem.RetroPhotographer}}
        }}]
      priority {curve {rowName=None}}
     }
    ]
   }
  }
 }
 ConversationLinkNode_0 {id=#4, class=ConversationLinkNode, properties {remoteEntryTag {tagName=Conversation.Entry.BasicServices}}}
 ConversationLinkNode_1 {id=#3, class=ConversationLinkNode, properties {remoteEntryTag {tagName=Conversation.Entry.NPCLibrary.Quests}}}
]
